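# Looping sound cue for the CubeIsland event's DudeBro vent, played while in the air, in the scripted events sound class.
# An IsLocal branch picks stereo status-effect attenuation for the local case and mono attenuation otherwise or when unset.
# Both paths loop the same vent wave through a random modulator and a continuous modulator whose FallSpeed parameter drives volume.
# The continuous modulator also sets a pitch range, with no parameter named for it.
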SoundCue /FNGameplayCues/Athena/Sounds/Events/CubeIsland/DudeBro_Vent_InAirLoop_Cue
root=Branch_0
nodes [
 {
  id=Attenuation_0
  children [ModulatorContinuous_2]
  properties {attenuationSettings=/FNGameplayCues/Sounds/Attenuation/Fort_Ability_Status_Effect_Looping_Stereo_Att}
 }
 {
  id=Attenuation_1
  children [ModulatorContinuous_0]
  properties {attenuationSettings=/FNGameplayCues/Sounds/Attenuation/Fort_Ability_Status_Effect_Looping_Mono_Att}
 }
 {id=Branch_0, children [Attenuation_0, Attenuation_1, Attenuation_1], properties {boolParameterName=IsLocal}}
 {
  id=ModulatorContinuous_0
  children [Modulator_1]
  properties {pitchModulationParams {maxOutput=1.25, minOutput=1}, volumeModulationParams {minOutput=0.25, parameterName=FallSpeed}}
 }
 {
  id=ModulatorContinuous_2
  children [Modulator_0]
  properties {pitchModulationParams {maxOutput=1.25, minOutput=1}, volumeModulationParams {minOutput=0.25, parameterName=FallSpeed}}
 }
 {id=Modulator_0, children [WavePlayer_0]}
 {id=Modulator_1, children [WavePlayer_1]}
 {id=WavePlayer_0, properties {bLooping=true, soundWaveAssetPtr=/Game/Athena/Sounds/Events/CubeIsland/DudeBro_Vent_Loop_01}}
 {id=WavePlayer_1, properties {bLooping=true, soundWaveAssetPtr=&DudeBro_Vent_Loop_01}}
]
properties {maxDistance=3500, pitchMultiplier=1.2, soundClassObject=/Game/Sounds/FortSoundClassesAndModes/Fort_Events_Scripted, volumeMultiplier=0.7018}
