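# Sound cue named for a DudeBro vortex fly effect in the Athena CubeIsland event sounds, with no root node or sound nodes listed, so it references no audio.
# It carries playback settings only: a status-effect activation attenuation, the scripted-events sound class, and adjusted volume and pitch.
# Its concurrency override allows a single instance and blocks new plays while one is active.
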SoundCue /FNGameplayCues/Athena/Sounds/Events/CubeIsland/DudeBro_Vortex_Fly_Cue
root=null
nodes []
properties {
 attenuationSettings=/Game/Sounds/Attenuation/Fort_Ability_Status_Effect_Activation_Att
 bOverrideConcurrency=true
 maxDistance=6500
 pitchMultiplier=1.2
 soundClassObject=/Game/Sounds/FortSoundClassesAndModes/Fort_Events_Scripted
 volumeMultiplier=0.2358
 concurrencyOverrides {maxCount=1, resolutionRule=PreventNew}
}
